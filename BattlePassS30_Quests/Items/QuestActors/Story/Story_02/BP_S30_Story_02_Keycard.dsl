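; Season 30 story quest pickup for the Refinery Keycard, collected when a player walks into it.
; Capsule overlap casts the other actor to a Fort player pawn and checks whether it can interact.
; With authority, it broadcasts an overlapped-quest-item custom verb message tagged for the keycard quest, then calls InteractHideandDestroy with that player's controller.
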(event EventBeginPlay
  (|CallParentFunction) ; This node is disabled and will not be called.\nDrag off pins to build functionality.
)

(event Collision|EventActorBeginOverlap
  (params (OtherActor :type Actor%20Object%20Reference)))

(event EventTick
  (params (DeltaSeconds :type Float%20%28single-precision%29)))

(event OnComponentBeginOverlap(Capsule)
  (params (OverlappedComponent :type Primitive%20Component%20Object%20Reference) (OtherActor :type Actor%20Object%20Reference) (OtherComp :type Primitive%20Component%20Object%20Reference) (OtherBodyIndex :type Integer) (bFromSweep :type Boolean) (SweepResult :type Hit%20Result%20Structure%20%28by%20ref%29))
  (bind _asfort_player_pawn (Utilities|Casting|CastToFortPlayerPawn OtherActor))
  (if (Interaction|BlueprintCanInteract _asfort_player_pawn "FirstInteraction" "IT_NoInteraction")
    (if (Networking|HasAuthority _asfort_player_pawn)
      (bind _result (CustomVerb|AddTagsforSubject (CustomVerb|MakeCustomVerbMessage _asfort_player_pawn "(TagName=\"CustomVerb.OverlappedQuestItem\")") "Refinery Keycard" (GameplayTags|MakeGameplayTagContainerfromTag "(TagName=\"Athena.Quests.S30.Story.02.Keycard\")")))
      (CustomVerb|BroadcastCustomVerbMessage _result)
      (CallFunction|InteractHideandDestroy (Fortnite|GetFortPlayerControllerfromActor _asfort_player_pawn)))))

(fn ConstructionScript
  (|CallParentFunction))
